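# Sound cue for a dojo drum's death, layering a drum hit with a wood-breaking crunch in the impacts sound class.
# A random drum impact passes through an enveloper that drops its pitch from above to well below normal and fades it out.
# A mixer blends it with a slightly quieter, pitched-down random torch wood-break sound borrowed from the Husk Pitcher villager projectile set.
# Distance falloff reuses the Cedar Woods chime impact attenuation.
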SoundCue /Game/Sounds/Fort_Impact_Sounds/DojoDrum/Dojo_Drum_Death_Cue
root=Mixer_1
nodes [
 {
  id=Enveloper_0
  children [Random_0]
  properties {
   pitchCurve {editorCurveData {keys [<at 0.0105 1.3206>,<at 0.2888 1.1858>,<at 0.4135 0.6956>,<at 0.5958 0.5117>]}}
   volumeCurve {editorCurveData {keys [<at 0 1>,<at 1 0.9919>,<at 1.5 0>]}}
  }
 }
 {id=Mixer_1, children [Enveloper_0,Modulator_1], properties {inputVolume [1,0.8]}}
 {id=Modulator_1, children [Random_1], properties {pitchMax=0.75, pitchMin=0.7}}
 {
  id=Random_0
  children [WavePlayer_0,WavePlayer_1,WavePlayer_2,WavePlayer_3,WavePlayer_4,WavePlayer_5,WavePlayer_6]
  properties {bShouldExcludeFromBranchCulling=true, weights [1,1,1,1,1,1,1]}
 }
 {
  id=Random_1
  children [WavePlayer_9, WavePlayer_8, WavePlayer_10, WavePlayer_11]
  properties {bShouldExcludeFromBranchCulling=true, weights [1,1,1,1]}
 }
 {id=WavePlayer_0, properties {soundWaveAssetPtr=&/Dojo_Drum_Impact_01}}
 {id=WavePlayer_1, properties {soundWaveAssetPtr=&/Dojo_Drum_Impact_02}}
 {
  id=WavePlayer_10
  properties {soundWaveAssetPtr=/Game/Sounds/Fort_Monster_Husk_Pitcher_Villager/Proj_Impact_Villager_Husk_Pitcher_Torch_Wood_Break_03_}
 }
 {
  id=WavePlayer_11
  properties {soundWaveAssetPtr=/Game/Sounds/Fort_Monster_Husk_Pitcher_Villager/Proj_Impact_Villager_Husk_Pitcher_Torch_Wood_Break_04_}
 }
 {id=WavePlayer_2, properties {soundWaveAssetPtr=&/Dojo_Drum_Impact_03}}
 {id=WavePlayer_3, properties {soundWaveAssetPtr=&/Dojo_Drum_Impact_04}}
 {id=WavePlayer_4, properties {soundWaveAssetPtr=&/Dojo_Drum_Impact_05}}
 {id=WavePlayer_5, properties {soundWaveAssetPtr=&/Dojo_Drum_Impact_06}}
 {id=WavePlayer_6, properties {soundWaveAssetPtr=&/Dojo_Drum_Impact_07}}
 {
  id=WavePlayer_8
  properties {soundWaveAssetPtr=/Game/Sounds/Fort_Monster_Husk_Pitcher_Villager/Proj_Impact_Villager_Husk_Pitcher_Torch_Wood_Break_02_}
 }
 {
  id=WavePlayer_9
  properties {soundWaveAssetPtr=/Game/Sounds/Fort_Monster_Husk_Pitcher_Villager/Proj_Impact_Villager_Husk_Pitcher_Torch_Wood_Break_01_}
 }
]
properties {
 attenuationSettings=/Game/Sounds/Fort_Ambient_Looping/CedarForestChime/CedarWoods_Chime_Impact_Att
 maxDistance=4800
 soundClassObject=&Fort_Impacts
 volumeMultiplier=2.5
}
