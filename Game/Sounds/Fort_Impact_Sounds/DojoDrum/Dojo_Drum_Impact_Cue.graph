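# Sound cue for dojo drum impacts, playing a varied drum hit through the Fort_Impacts sound class.
# A Random node picks one of seven equally weighted drum impact waves and is excluded from branch culling.
# A Modulator above it adds slight volume variation, and the cue boosts overall volume.
# Spatialization uses an attenuation asset borrowed from the CedarWoods chime impact sounds.
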SoundCue /Game/Sounds/Fort_Impact_Sounds/DojoDrum/Dojo_Drum_Impact_Cue
root=Modulator_0
nodes [
 {id=Modulator_0, children [Random_0], properties {volumeMin=0.9}}
 {
  id=Random_0
  children [WavePlayer_0,WavePlayer_1,WavePlayer_2,WavePlayer_3,WavePlayer_4,WavePlayer_5,WavePlayer_6]
  properties {bShouldExcludeFromBranchCulling=true, weights [1,1,1,1,1,1,1]}
 }
 {id=WavePlayer_0, properties {soundWaveAssetPtr=&/Dojo_Drum_Impact_01}}
 {id=WavePlayer_1, properties {soundWaveAssetPtr=&/Dojo_Drum_Impact_02}}
 {id=WavePlayer_2, properties {soundWaveAssetPtr=&/Dojo_Drum_Impact_03}}
 {id=WavePlayer_3, properties {soundWaveAssetPtr=&/Dojo_Drum_Impact_04}}
 {id=WavePlayer_4, properties {soundWaveAssetPtr=&/Dojo_Drum_Impact_05}}
 {id=WavePlayer_5, properties {soundWaveAssetPtr=&/Dojo_Drum_Impact_06}}
 {id=WavePlayer_6, properties {soundWaveAssetPtr=&/Dojo_Drum_Impact_07}}
]
properties {
 attenuationSettings=/Game/Sounds/Fort_Ambient_Looping/CedarForestChime/CedarWoods_Chime_Impact_Att
 maxDistance=4800
 soundClassObject=&Fort_Impacts
 volumeMultiplier=2
}
